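# Looping interior rumble ambience cue for the Doom Forge at the CastleForge POI, built from reused meteor, sky island and castle room loops.
# A quality switch picks between two three-layer mixers that differ only in quad versus stereo room tone.
# Two nested, endlessly looping envelopers add slow drift, the inner varying pitch over 30 seconds and the outer gently varying volume over 20.
# Attenuation overrides set a custom falloff with distance-based low-pass filtering.
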SoundCue /Game/Sounds/Helios/POI/CastleForge/Ambience_Doom_Forge_Main_Rumble_Interior_Cue
root=Enveloper_2
nodes [
 {
  id=Enveloper_1
  children [QualityLevel_0]
  properties {
   bLoop=true
   bLoopIndefinitely=true
   loopEnd=30
   pitchCurve {editorCurveData {keys [{interpMode=RCIM_Cubic, time=0}, <at 15.9468 0.9181 RCIM_Cubic>, {interpMode=RCIM_Cubic, time=30}]}}
   volumeCurve {editorCurveData {keys [<at 0 1>]}}
  }
 }
 {
  id=Enveloper_2
  children [Enveloper_1]
  properties {
   bLoop=true
   bLoopIndefinitely=true
   loopEnd=20
   pitchCurve {editorCurveData {keys [{time=0}]}}
   volumeCurve {editorCurveData {keys [
      {arriveTangent=0, interpMode=RCIM_Cubic, leaveTangent=0, time=0, value=1}
      {arriveTangent=0, interpMode=RCIM_Cubic, leaveTangent=0, time=10, value=0.9}
      {arriveTangent=0, interpMode=RCIM_Cubic, leaveTangent=0, time=20, value=0.9992}
     ]}}
  }
 }
 {id=Mixer_1, children [WavePlayer_5,WavePlayer_2,WavePlayer_0], properties {inputVolume [0.8,0.4,1]}}
 {id=Mixer_3, children [WavePlayer_12, WavePlayer_3, WavePlayer_1], properties {inputVolume [0.8,0.4,1]}}
 {id=QualityLevel_0, children [Mixer_1,Mixer_1,Mixer_3,Mixer_3]}
 {id=WavePlayer_0, properties {bLooping=true, soundWaveAssetPtr=/Game/Sounds/Asteria/POI/Knights/Ambience_Knights_Castle_Medium_Room_01_Quad}}
 {
  id=WavePlayer_1
  properties {bLooping=true, soundWaveAssetPtr=/Game/Sounds/Asteria/POI/Knights/Ambience_Knights_Castle_Medium_Room_01_Stereo}
 }
 {id=WavePlayer_12, properties {bLooping=true, soundWaveAssetPtr=/Game/Sounds/Asteria/POI/Meteor/Spots/Spot_Meteor_Main_Rumble_Loop}}
 {id=WavePlayer_2, properties {bLooping=true, soundWaveAssetPtr=/Game/Sounds/Asteria/POI/SkyIsland/Spots/Spot_SkyIsland_Rumble_All}}
 {id=WavePlayer_3, properties {bLooping=true, soundWaveAssetPtr=&Spot_SkyIsland_Rumble_All}}
 {id=WavePlayer_5, properties {bLooping=true, soundWaveAssetPtr=&Spot_Meteor_Main_Rumble_Loop}}
]
properties {
 attenuationSettings=&Fort_Ambient_Athena_Att
 maxDistance=2097152
 soundClassObject=&Fort_Ambient_Athena
 volumeMultiplier=0.626
 attenuationOverrides {
  bAttenuateWithLPF=true
  bEnableLogFrequencyScaling=true
  distanceAlgorithm=Custom
  falloffDistance=3000
  lPFFrequencyAtMax=197.5057
  lPFFrequencyAtMin=20511.4102
  lPFRadiusMax=3789.3618
  lPFRadiusMin=1822.3112
  nonSpatializedRadiusEnd=1500
  nonSpatializedRadiusStart=3000
  spatializationAlgorithm=SPATIALIZATION_HRTF
  stereoSpread=1000
  attenuationShapeExtents {x=0}
  customAttenuationCurve {editorCurveData {keys [<at 0 1>,<at 0.75 1>,<at 2 0>]}}
 }
}
