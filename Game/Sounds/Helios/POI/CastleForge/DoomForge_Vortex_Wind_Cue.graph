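# Looping wind ambience cue for the DoomForge vortex at the CastleForge POI in Helios.
# A single looping wave player reuses the CoolingTower wind sound under the Athena indoor ambient sound class.
# Overridden attenuation uses a large capsule shape with stereo spread, a non-spatialized inner radius, listener focus that lowers unfocused volume, and occlusion on a custom trace channel.
# The debug flag is left on.
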SoundCue /Game/Sounds/Helios/POI/CastleForge/DoomForge_Vortex_Wind_Cue
root=WavePlayer_0
nodes [{id=WavePlayer_0, properties {bLooping=true, soundWaveAssetPtr=/Game/Sounds/Fort_Prop_Sounds/Exterior/CoolingTower/CoolingTowerWind_01}}]
properties {
 bDebug=true
 bOverrideAttenuation=true
 maxDistance=8800
 soundClassObject=&Fort_Ambient_Indoors_Athena
 volumeMultiplier=1.002
 attenuationOverrides {
  attenuationShape=Capsule
  bEnableFocusInterpolation=true
  bEnableListenerFocus=true
  falloffDistance=1000
  nonFocusVolumeAttenuation=0.7
  nonSpatializedRadiusStart=1500
  occlusionLowPassFilterFrequency=10000
  occlusionTraceChannel=ECC_GameTraceChannel11
  occlusionVolumeAttenuation=0.5
  stereoSpread=1000
  attenuationShapeExtents=<7800 2000>
 }
}
